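; War effort config data class that exposes lookups for funding limits and initiatives by gameplay tag.
; GetFundingLimitByFundingType scans the funding limit array for an exact tag match and returns that entry's scalable-float limit evaluated to a number.
; GetInitiativeByTag scans the initiatives array the same way and returns the matching config struct.
; Both return a found flag, which is false when nothing matches.
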(fn GetFundingLimitByFundingType
  (params (FundingTypeTag :type Gameplay%20Tag%20Structure))
  (bind _array_element (Utilities|Array|ForEachLoopwithBreak (Variables|Default|GetFundingLimits))
    (:LoopBody
      (if (GameplayTags|MatchesTag (Utilities|Struct|BreakFWarEffortFundingLimitConfig _array_element) FundingTypeTag true)
        (return true (ScalableFloat|GetValueatLevel (.FundingLimit_5_EB3C9D2149584DE78508C1883DE7AEB0 (Utilities|Struct|BreakFWarEffortFundingLimitConfig _array_element))))))
    (:Completed
      (return false 0.0))))

(fn GetInitiativeByTag
  (params (InitiativeTag :type Gameplay%20Tag%20Structure))
  (for _array_element (Variables|Default|GetInitiatives)
    (if (GameplayTags|MatchesTag (Utilities|Struct|BreakFWarEffortInitiativeConfig _array_element) InitiativeTag true)
      (return true _array_element)))
  (return false 0))
